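# Athena loading screen preview panel widget that shows a cosmetic loading screen full-screen with a Close button.
# The native base AthenaLoadingScreenPreviewPanelBase supplies the preview host slot, layered over a solid color image, and the button, bound to the Back input action in a safe zone.
# Close calls HandleBack, which runs an empty PopDetails and returns false.
# Init and RegisterInput have no connected logic; UnregisterInput removes all input handlers.
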WidgetBlueprint /Game/Athena/Cosmetics/LoadingScreens/B_AthenaLoadingScreenPreviewPanel
graphs [
 EventGraph {nodes [
   "On Clicked (Button_Close)"!() -> (OutputDelegate, then, Button:Common Button Base Object Reference) {is=title K2Node_ComponentBoundEvent}
   > HandleBack!(execute, self) -> (then, PassThrough:bool)
   Event On Activated()
   > Init()
   Event Construct()
   > RegisterInput()
  ]}
 HandleBack {nodes [
   HandleBack() {is=title K2Node_FunctionEntry}
   > PopDetails()
   > Return Node!(execute, PassThrough:bool=false)
  ]}
 Init {nodes [
   Init() {is=title K2Node_FunctionEntry}
   Get BackgroundWidget!(self:Athena Loading Screen Item Definition Object Reference=Get LoadingScreenAsset() -> (LoadingScreenAsset:Athena Loading Screen Item Definition Object Reference)) -> (BackgroundWidget:User Widget Soft Class Reference)
  ]}
 PopDetails {nodes [PopDetails() {is=title K2Node_FunctionEntry}]}
 RegisterInput {nodes [RegisterInput() {is=title K2Node_FunctionEntry}]}
 UnregisterInput {nodes [
   UnregisterInput() {is=title K2Node_FunctionEntry}
   > RemoveAllInputActionHandlers(self:Common Activatable Panel Legacy Object Reference)
  ]}
]
variables [LoadingScreenAsset {type=AthenaLoadingScreenItemDefinition, default=null}]
widgets {
 info {
  inheritedWidgetCount=3
  parentClass=/Script/FortniteUI.AthenaLoadingScreenPreviewPanelBase
  rootWidgetClass=&Overlay
  widgetCount=6
 }
 widgets [
  Overlay_2 {widgetClassPath=&Overlay, widget=Overlay_2}
  BaseColor {
   widgetClassPath=&Image
   bIsVariable=true
   parent=Overlay_2
   widget=BaseColor
   properties {colorAndOpacity=<rgba 0.6445 0.0222 0 1>}
   slot {class=OverlaySlot, properties {horizontalAlignment=HAlign_Fill, verticalAlignment=VAlign_Fill}}
  }
  LoadingScreenSlot {
   widgetClassPath=/Script/UMG.NativeWidgetHost
   bInherited=true
   bIsVariable=true
   parent=Overlay_2
   widget=LoadingScreenSlot
   slot {class=OverlaySlot, properties {horizontalAlignment=HAlign_Fill, verticalAlignment=VAlign_Fill}}
  }
  SafeZone_0 {
   widgetClassPath=&SafeZone
   bIsVariable=true
   parent=Overlay_2
   widget=SafeZone_0
   properties {padLeft=false, padTop=false}
   slot {class=OverlaySlot, properties {horizontalAlignment=HAlign_Fill, verticalAlignment=VAlign_Fill, padding=<ltrb 0 0 60 175>}}
  }
  Button_Close {
   widgetClassPath=/FortUILibrary/UIKit/Buttons/WBP_UIKit_Button_Loud
   bInherited=true
   bIsVariable=true
   parent=SafeZone_0
   widget=Button_Close
   properties {priority=1, text=Close, triggeringInputAction {dataTable=&InputActionDataTable, rowName=Back}}
   slot {class=SafeZoneSlot, properties {hAlign=HAlign_Right, vAlign=VAlign_Bottom}}
  }
  Button_CloseTouch {widgetClassPath=&CommonButtonBase, bInherited=true, widget=null}
 ]
}
