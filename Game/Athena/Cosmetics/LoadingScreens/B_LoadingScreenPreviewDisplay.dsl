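; Frontend display actor that previews a loading screen cosmetic on a static mesh in item inspection.
; On the preview event it rescales the pivot holder, assigns the preview mesh and feeds the item's background image into a dynamic material's WorkerPortrait texture parameter.
; The mesh hides while a character-transition particle effect spawns, then reappears after a short delay.
; It also supplies a special preview action for viewing the loading screen.
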(event LoadingScreen|EventPreviewLoadingScreen
  (params (LoadingScreen :type Athena%20Loading%20Screen%20Item%20Definition%20Object%20Reference))
  (bind _staticmesh (Variables|Default|GetStaticMesh))
  (Transformation|SetWorldTransform (Variables|Default|GetPivotHolder) (Math|Transform|MakeTransform (.location (Transformation|GetActorTransform)) (.rotation (Transformation|GetActorTransform)) (* (.scale (Transformation|GetActorTransform)) 0.9)) false true)
  (Components|StaticMesh|SetStaticMesh _staticmesh "/Game/UI/Frontend/ItemInspect/LoadingScreens/S_LoadingScreenPreview.S_LoadingScreenPreview")
  (bind _returnvalue (Rendering|Material|CreateDynamicMaterialInstance _staticmesh 0 (Rendering|Material|GetMaterial _staticmesh 0) "None" :!class /Script/Engine.PrimitiveComponent))
  (Rendering|Material|SetTextureParameterValue _returnvalue "WorkerPortrait" (Utilities|ResolveSoftReference (Variables|/Script/FortniteGame.AthenaLoadingScreenItemDefinition|GetBackgroundImage LoadingScreen)))
  (Rendering|SetVisibility _staticmesh false false)
  (Utilities|FlowControl|Delay 0.0)
  (Effects|SpawnEmitteratLocation "/Game/Effects/Fort_Effects/UI/Transition/P_FrontEndCharacterTransition.P_FrontEndCharacterTransition" (Transformation|GetActorLocation) "0,90.000000, 0" "3.000000,3.000000,3.000000")
  (Utilities|FlowControl|Delay 0.1)
  (Rendering|SetVisibility _staticmesh true false))

(fn GetSpecialPreviewActions
  (return (Utilities|Array|MakeArray "SpecialPreviewAction_ViewLoadingScreen")))

(fn ConstructionScript)
